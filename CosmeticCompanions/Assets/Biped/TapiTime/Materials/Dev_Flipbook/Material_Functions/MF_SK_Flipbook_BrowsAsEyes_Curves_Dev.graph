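# Development material function for the TapiTime companion that selects an eye flipbook frame Index, with brow expressions folded into the eye frames.
# Animation-driven, right-side brow, blink and lid curves plus an emote curve are compared against tunable threshold parameters.
# A nested If chain returns constant indices for neutral, closed, down, angry, worried, squint and custom eyes.
# Min of curves requires combined poses, while subtracted terms and Step-based override masks arbitrate between overlapping expressions.
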MaterialFunction /CosmeticCompanions/Assets/Biped/TapiTime/Materials/Dev_Flipbook/Material_Functions/MF_SK_Flipbook_BrowsAsEyes_Curves_Dev
graph {nodes [
  FunctionOutput(None=NamedRerouteUsage()) {outputName=Index}
  NamedRerouteDeclaration(None=ScalarParameter() {group="Flipbook 03.01 Eyes | Curves (Driven By Animation)", parameterName=R_brow_squeeze_in_pose, sortPriority=7}) {name=Brow_Squeeze_In_Pose, nodeColor=<rgba 1 0.2105 0 1>}
  NamedRerouteDeclaration(None=ScalarParameter() {defaultValue=0.1, group=Flipbook 03.02 Eyes | Thresholds, parameterName=Threshold_Brow_Squeeze_In, sortPriority=7}) {name=Threshold_Brow_Squeeze_In, nodeColor=<rgba 1 0.0834 0.4399 1>}
  NamedRerouteDeclaration(None=Constant()) {name=Index_Neutral_CC, nodeColor=<rgba 0.2818 0.3488 1 1>}
  NamedRerouteDeclaration(None=ScalarParameter() {group="Flipbook 03.01 Eyes | Curves (Driven By Animation)", parameterName=R_brow_In_Dn_pose, sortPriority=0}) {name=Brow_In_Dn_Pose, nodeColor=<rgba 1 0.2105 0 1>}
  NamedRerouteDeclaration(None=ScalarParameter() {group="Flipbook 03.01 Eyes | Curves (Driven By Animation)", parameterName=R_brow_Ot_Dn_pose, sortPriority=2}) {name=Brow_Ot_Dn_Pose, nodeColor=<rgba 1 0.2105 0 1>}
  NamedRerouteDeclaration(None=Constant() {r=3}) {name=Eyes Angry CC, nodeColor=<rgba 0.2818 0.3488 1 1>}
  NamedRerouteDeclaration(None=ScalarParameter() {group="Flipbook 03.01 Eyes | Curves (Driven By Animation)", parameterName=eye_emote_01, sortPriority=7}) {name=Eye_Emote_01_Pose, nodeColor=<rgba 1 0.2105 0 1>}
  NamedRerouteDeclaration(None=Constant() {r=1}) {name=Eyes Closed, nodeColor=<rgba 0.2818 0.3488 1 1>}
  NamedRerouteDeclaration(None=Constant() {r=2}) {name=Eyes Down CC, nodeColor=<rgba 0.2818 0.3488 1 1>}
  NamedRerouteDeclaration(None=ScalarParameter() {defaultValue=0.1, group=Flipbook 03.02 Eyes | Thresholds, parameterName=Threshold_Brow_In_Dn, sortPriority=0}) {name=Threshold_Brow_In_Dn, nodeColor=<rgba 1 0.0834 0.4399 1>}
  NamedRerouteDeclaration(None=ScalarParameter() {defaultValue=0.1, group=Flipbook 03.02 Eyes | Thresholds, parameterName=Threshold_Brow_Ot_Dn, sortPriority=2}) {name=Threshold_Brow_Ot_Dn, nodeColor=<rgba 1 0.0834 0.4399 1>}
  NamedRerouteDeclaration(None=ScalarParameter() {group="Flipbook 03.01 Eyes | Curves (Driven By Animation)", parameterName=R_brow_In_Up_pose, sortPriority=1}) {name=Brow_In_Up_Pose, nodeColor=<rgba 1 0.2105 0 1>}
  NamedRerouteDeclaration(None=ScalarParameter() {defaultValue=0.1, group=Flipbook 03.02 Eyes | Thresholds, parameterName=Threshold_Brow_In_Up, sortPriority=1}) {name=Threshold_Brow_In_Up, nodeColor=<rgba 1 0.0834 0.4399 1>}
  NamedRerouteDeclaration(None=ScalarParameter() {group="Flipbook 03.01 Eyes | Curves (Driven By Animation)", parameterName=R_brow_Ot_Up_pose, sortPriority=3}) {name=Brow_Ot_Up_Pose, nodeColor=<rgba 1 0.2105 0 1>}
  NamedRerouteDeclaration(None=ScalarParameter() {defaultValue=0.1, group=Flipbook 03.02 Eyes | Thresholds, parameterName=Threshold_Brow_Ot_Up, sortPriority=3}) {name=Threshold_Brow_Ot_Up, nodeColor=<rgba 1 0.0834 0.4399 1>}
  NamedRerouteDeclaration(None=Constant() {r=4}) {name=Eyes Angry Squint CC, nodeColor=<rgba 0.2818 0.3488 1 1>}
  NamedRerouteDeclaration(None=ScalarParameter() {defaultValue=0.1, group=Flipbook 03.02 Eyes | Thresholds, parameterName=Threshold_Eye_Emote_01, sortPriority=7}) {name=Threshold_Eye_Emote_01, nodeColor=<rgba 1 0.0834 0.4399 1>}
  NamedRerouteDeclaration(None=ScalarParameter() {group="Flipbook 03.01 Eyes | Curves (Driven By Animation)", parameterName=R_blink_pose, sortPriority=4}) {name=Blink_Pose, nodeColor=<rgba 1 0.2105 0 1>}
  NamedRerouteDeclaration(None=ScalarParameter() {group="Flipbook 03.01 Eyes | Curves (Driven By Animation)", parameterName=R_lower_lid_up_pose, sortPriority=6}) {name=Lower_Lid_Up_Pose, nodeColor=<rgba 1 0.2105 0 1>}
  NamedRerouteDeclaration(None=Constant() {r=7}) {name=Custom Eyes, nodeColor=<rgba 0.2818 0.3488 1 1>}
  NamedRerouteDeclaration(None=Constant() {r=5}) {name=Eyes Worried CC, nodeColor=<rgba 0.2818 0.3488 1 1>}
  NamedRerouteDeclaration(None=ScalarParameter() {defaultValue=0.1, group=Flipbook 03.02 Eyes | Thresholds, parameterName=Threshold_Blink, sortPriority=4}) {name=Threshold_Blink, nodeColor=<rgba 1 0.0834 0.4399 1>}
  n1 = Min(A=Min(A=NamedRerouteUsage(), B=NamedRerouteUsage()), B=NamedRerouteUsage())
  n2 = NamedRerouteUsage()
  n3 = Max(A=NamedRerouteUsage(), B=NamedRerouteUsage())
  n4 = NamedRerouteUsage()
  n5 = NamedRerouteUsage()
  n6 = NamedRerouteUsage()
  n7 = Subtract(A=Subtract(A=Min(A=@n5, B=@n6), B=NamedRerouteUsage()), B=NamedRerouteUsage())
  n8 = NamedRerouteUsage()
  n9 = Min(A=NamedRerouteUsage(), B=NamedRerouteUsage())
  n10 = Subtract(A=@n9, B=NamedRerouteUsage())
  n11 = NamedRerouteUsage()
  n12 = NamedRerouteUsage()
  n13 = NamedRerouteUsage()
  n14 = Min(A=Min(A=NamedRerouteUsage(), B=NamedRerouteUsage()), B=NamedRerouteUsage())
  n15 = Subtract(A=@n14, B=NamedRerouteUsage())
  n16 = NamedRerouteUsage()
  n17 = NamedRerouteUsage()
  n18 = NamedRerouteUsage()
  n19 = Max(A=@n17, B=@n18)
  n20 = NamedRerouteUsage()
  n21 = Min(A=NamedRerouteUsage(), B=NamedRerouteUsage())
  n22 = Subtract(A=@n21, B=NamedRerouteUsage())
  n23 = NamedRerouteUsage()
  n24 = Min(A=Min(A=NamedRerouteUsage(), B=NamedRerouteUsage()), B=NamedRerouteUsage())
  n25 = NamedRerouteUsage()
  n26 = NamedRerouteUsage()
  n27 = NamedRerouteUsage()
  n28 = If(
   A=@n24
   B=Max(A=Max(A=@n25, B=@n26), B=@n27)
   "A > B"=NamedRerouteUsage()
   "A < B"=If(A=NamedRerouteUsage(), B=NamedRerouteUsage(), "A > B"=NamedRerouteUsage(), "A < B"=NamedRerouteUsage())
  )
  n29 = NamedRerouteUsage()
  n30 = NamedRerouteUsage()
  n31 = Max(A=@n29, B=@n30)
  n32 = If(
   A=Subtract(A=Subtract(A=@n10, B=@n11), B=NamedRerouteUsage()) {constB=0}
   B=Max(A=@n12, B=@n13)
   "A > B"=NamedRerouteUsage()
   "A < B"=If(A=Subtract(A=@n15, B=@n16), B=Max(A=@n19, B=@n20), "A > B"=NamedRerouteUsage(), "A < B"=If(A=@n22, B=@n31, "A > B"=@n23, "A < B"=@n28))
  )
  n33 = Max(A=NamedRerouteUsage(), B=NamedRerouteUsage())
  NamedRerouteDeclaration(None=If(A=Subtract(A=@n1, B=@n2), B=Max(A=@n3, B=@n4), "A > B"=NamedRerouteUsage(), "A < B"=If(A=@n7, B=@n33, "A > B"=@n8, "A < B"=@n32))) {name=Eyes as Brows Output, nodeColor=<rgba 1 0.5859 0 1>}
  NamedRerouteDeclaration(None=ScalarParameter() {defaultValue=0.1, group=Flipbook 03.02 Eyes | Thresholds, parameterName=Threshold_Lower_Lid_Up, sortPriority=6}) {name=Threshold_Lower_Lid_Up, nodeColor=<rgba 1 0.0834 0.4399 1>}
  NamedRerouteDeclaration(None=ScalarParameter() {group="Flipbook 03.01 Eyes | Curves (Driven By Animation)", parameterName=R_upperLidDown_pose, sortPriority=7}) {name=Upper_Lid_Down_Pose, nodeColor=<rgba 1 0.2105 0 1>}
  NamedRerouteDeclaration(None=ScalarParameter() {defaultValue=0.1, group=Flipbook 03.02 Eyes | Thresholds, parameterName=Threshold_Upper_Lid_Down, sortPriority=7}) {name=Threshold_Upper_Lid_Down, nodeColor=<rgba 1 0.0834 0.4399 1>}
  NamedRerouteDeclaration(None=ScalarParameter() {group="Flipbook 03.01 Eyes | Curves (Driven By Animation)", parameterName=R_wide_pose, sortPriority=5}) {name=Wide_Pose, nodeColor=<rgba 1 0.2105 0 1>}
  NamedRerouteDeclaration(None=ScalarParameter() {defaultValue=0.1, group=Flipbook 03.02 Eyes | Thresholds, parameterName=Threshold_Wide, sortPriority=5}) {name=Threshold_Wide, nodeColor=<rgba 1 0.0834 0.4399 1>}
  NamedRerouteDeclaration(None=Constant() {r=6}) {name=Eyes Worried Squint CC, nodeColor=<rgba 0.2818 0.3488 1 1>}
  n34 = Min(A=@n25, B=@n26)
  NamedRerouteDeclaration(None=Step(Y=Min(A=@n34, B=@n27), X=@n24) {constY=1}) {name=Worried Squint CC Override, nodeColor=<rgba 1 0 0.2344 1>}
  NamedRerouteDeclaration(None=Step(Y=Min(A=@n29, B=@n30), X=@n21) {constY=1}) {name=Worried CC Override, nodeColor=<rgba 1 0 0.1641 1>}
  n35 = Min(A=@n17, B=@n18)
  NamedRerouteDeclaration(None=Step(Y=Min(A=@n35, B=@n20), X=@n14) {constY=1}) {name=Eyes Angry Squint CC Override, nodeColor=<rgba 0 1 0.4141 1>}
  NamedRerouteDeclaration(None=Step(Y=Min(A=@n12, B=@n13), X=@n9) {constY=1}) {name=Eyes Angry CC Override, nodeColor=<rgba 0.0312 0 1 1>}
 ]}
