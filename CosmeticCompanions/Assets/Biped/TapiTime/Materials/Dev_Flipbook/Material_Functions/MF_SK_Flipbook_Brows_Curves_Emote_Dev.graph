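# Dev material function for the TapiTime cosmetic companion that outputs a brow flipbook frame Index from animation-driven curves.
# Scalar parameters carry right-brow inner/outer up and down poses plus an emote curve, each with a tunable threshold.
# Five If nodes, wired through named reroutes, appear to compare curves to thresholds and select among constant indices for neutral, four poses and the emote.
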MaterialFunction /CosmeticCompanions/Assets/Biped/TapiTime/Materials/Dev_Flipbook/Material_Functions/MF_SK_Flipbook_Brows_Curves_Emote_Dev
graph {nodes [
  FunctionOutput(None=NamedRerouteUsage()) {outputName=Index}
  NamedRerouteDeclaration(None=ScalarParameter() {group="Flipbook 02.01 Brows | Curves (Driven By Animation)", parameterName=R_brow_Ot_Up_pose, sortPriority=3}) {name=Brow_Ot_Up_Pose, nodeColor=<rgba 1 0.2105 0 1>}
  NamedRerouteDeclaration(None=ScalarParameter() {defaultValue=0.1, group=Flipbook 02.02 Brows | Thresholds, parameterName=Threshold_Brow_Ot_Up, sortPriority=3}) {name=Threshold_Brow_Ot_Up, nodeColor=<rgba 1 0.0834 0.4399 1>}
  NamedRerouteDeclaration(None=Constant() {r=5}) {name=Index_Brow_Emote_01, nodeColor=<rgba 0.2818 0.3488 1 1>}
  NamedRerouteDeclaration(None=ScalarParameter() {group="Flipbook 02.01 Brows | Curves (Driven By Animation)", parameterName=R_brow_In_Dn_pose, sortPriority=0}) {name=Brow_In_Dn_Pose, nodeColor=<rgba 1 0.2105 0 1>}
  NamedRerouteDeclaration(None=ScalarParameter() {group="Flipbook 02.01 Brows | Curves (Driven By Animation)", parameterName=R_brow_Ot_Dn_pose, sortPriority=2}) {name=Brow_Ot_Dn_Pose, nodeColor=<rgba 1 0.2105 0 1>}
  NamedRerouteDeclaration(None=Constant()) {name=Index_Neutral, nodeColor=<rgba 0.2818 0.3488 1 1>}
  NamedRerouteDeclaration(None=Constant() {r=3}) {name=Index_Brow_Ot_Dn, nodeColor=<rgba 0.2818 0.3488 1 1>}
  NamedRerouteDeclaration(None=Constant() {r=1}) {name=Index_Brow_In_Dn, nodeColor=<rgba 0.2818 0.3488 1 1>}
  NamedRerouteDeclaration(None=Constant() {r=2}) {name=Index_Brow_In_Up, nodeColor=<rgba 0.2818 0.3488 1 1>}
  NamedRerouteDeclaration(None=ScalarParameter() {group="Flipbook 02.01 Brows | Curves (Driven By Animation)", parameterName=R_brow_In_Up_pose, sortPriority=1}) {name=Brow_In_Up_Pose, nodeColor=<rgba 1 0.2105 0 1>}
  NamedRerouteDeclaration(None=ScalarParameter() {defaultValue=0.1, group=Flipbook 02.02 Brows | Thresholds, parameterName=Threshold_Brow_In_Up, sortPriority=1}) {name=Threshold_Brow_In_Up, nodeColor=<rgba 1 0.0834 0.4399 1>}
  NamedRerouteDeclaration(None=Constant() {r=4}) {name=Index_Brow_Ot_Up, nodeColor=<rgba 0.2818 0.3488 1 1>}
  NamedRerouteDeclaration(None=If(A=NamedRerouteUsage(), B=NamedRerouteUsage(), "A > B"=NamedRerouteUsage(), "A < B"=NamedRerouteUsage())) {name=Brow_In_Dn_Output, nodeColor=<rgba 0 0.1532 1 1>}
  NamedRerouteDeclaration(None=If(A=NamedRerouteUsage(), B=NamedRerouteUsage(), "A > B"=NamedRerouteUsage(), "A < B"=NamedRerouteUsage())) {name=Brow_Emote_01_Output, nodeColor=<rgba 0 0.1532 1 1>}
  NamedRerouteDeclaration(None=If(A=NamedRerouteUsage(), B=NamedRerouteUsage(), "A > B"=NamedRerouteUsage(), "A < B"=NamedRerouteUsage())) {name=Brow_In_Up_Output, nodeColor=<rgba 0 0.1532 1 1>}
  NamedRerouteDeclaration(None=If(A=NamedRerouteUsage(), B=NamedRerouteUsage(), "A > B"=NamedRerouteUsage(), "A < B"=NamedRerouteUsage())) {name=Brow_Ot_Dn_Output, nodeColor=<rgba 0 0.1532 1 1>}
  NamedRerouteDeclaration(None=If(A=NamedRerouteUsage(), B=NamedRerouteUsage(), "A > B"=NamedRerouteUsage(), "A < B"=NamedRerouteUsage())) {name=Brow_Ot_Up_Output, nodeColor=<rgba 0 0.1532 1 1>}
  NamedRerouteDeclaration(None=ScalarParameter() {group="Flipbook 02.01 Brows | Curves (Driven By Animation)", parameterName=brow_emote_01, sortPriority=3}) {name=Brow_Emote_01, nodeColor=<rgba 1 0.2105 0 1>}
  NamedRerouteDeclaration(None=ScalarParameter() {defaultValue=0.1, group=Flipbook 02.02 Brows | Thresholds, parameterName=Threshold_Brow_Emote_01, sortPriority=3}) {name=Threshold_Brow_Emote_01, nodeColor=<rgba 1 0.0834 0.4399 1>}
  NamedRerouteDeclaration(None=ScalarParameter() {defaultValue=0.1, group=Flipbook 02.02 Brows | Thresholds, parameterName=Threshold_Brow_In_Dn, sortPriority=0}) {name=Threshold_Brow_In_Dn, nodeColor=<rgba 1 0.0834 0.4399 1>}
  NamedRerouteDeclaration(None=ScalarParameter() {defaultValue=0.1, group=Flipbook 02.02 Brows | Thresholds, parameterName=Threshold_Brow_Ot_Dn, sortPriority=2}) {name=Threshold_Brow_Ot_Dn, nodeColor=<rgba 1 0.0834 0.4399 1>}
 ]}
